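# Physics asset for the female Golden Skeleton pickaxe cosmetic, defining bodies for two short dynamic chains on the melee weapon mesh.
# Each chain runs from the root capsule through a small MSTR capsule to a larger, offset end capsule.
# Four constraints join these links with swing and twist set to Free, so the stored 45-degree limits do not restrict rotation.
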PhysicsAsset /BRCosmetics/Weapons/FORT_Melee/Pickaxe_The_Golden_Skeleton_Female/Meshes/The_Golden_Skeleton_Female_Physics
bodies {
 b_AR_Root [{shapeType=Capsule, length=3.9475, radius=3.0344}]
 dyn_chain_01 [{shapeType=Capsule, length=1.3133, radius=1.7267, center=<0 0 -2.0343>}]
 dyn_chain_01_MSTR [{shapeType=Capsule, length=0.505, radius=0.505}]
 dyn_chain_02 [{shapeType=Capsule, length=1.3133, radius=1.7267, center=<0 0 -2.0285>}]
 dyn_chain_2_MSTR [{shapeType=Capsule, length=0.505, radius=0.505}]
}
constraints [
 {
  bone1Name=dyn_chain_01_MSTR
  bone2Name=b_AR_Root
  swing1LimitDegrees=45
  swing1Motion=Free
  swing2LimitDegrees=45
  swing2Motion=Free
  twistLimitDegrees=45
  twistMotion=Free
 }
 {
  bone1Name=dyn_chain_2_MSTR
  bone2Name=b_AR_Root
  swing1LimitDegrees=45
  swing1Motion=Free
  swing2LimitDegrees=45
  swing2Motion=Free
  twistLimitDegrees=45
  twistMotion=Free
 }
 {
  bone1Name=dyn_chain_01
  bone2Name=dyn_chain_01_MSTR
  swing1LimitDegrees=45
  swing1Motion=Free
  swing2LimitDegrees=45
  swing2Motion=Free
  twistLimitDegrees=45
  twistMotion=Free
 }
 {
  bone1Name=dyn_chain_02
  bone2Name=dyn_chain_2_MSTR
  swing1LimitDegrees=45
  swing1Motion=Free
  swing2LimitDegrees=45
  swing2Motion=Free
  twistLimitDegrees=45
  twistMotion=Free
 }
]
